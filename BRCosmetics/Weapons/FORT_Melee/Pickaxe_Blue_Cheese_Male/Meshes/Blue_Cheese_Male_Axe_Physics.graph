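# Physics asset for the Blue Cheese Male pickaxe cosmetic weapon mesh, a two-handed melee axe.
# A long, thin capsule body covers the handle bone, and a small capsule sits on the dyn_collar bone.
# One constraint joins the collar to the handle with small swing and twist limits, apparently so the collar can move slightly as a secondary dynamic part.
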PhysicsAsset /BRCosmetics/Weapons/FORT_Melee/Pickaxe_Blue_Cheese_Male/Meshes/Blue_Cheese_Male_Axe_Physics
bodies {
 Melee_TwoHanded_Handle [{shapeType=Capsule, length=40, radius=2, center=<0 -3 16>, rotation=<rot 0 0 -10>}]
 dyn_collar [{shapeType=Capsule, length=0, radius=3, center=<0 0 1.5>}]
}
constraints [{bone1Name=dyn_collar, bone2Name=Melee_TwoHanded_Handle, swing1LimitDegrees=3, swing2LimitDegrees=15, twistLimitDegrees=3}]
